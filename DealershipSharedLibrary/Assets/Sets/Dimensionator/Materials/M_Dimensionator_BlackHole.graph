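# Unlit, additive, two-sided material for a black hole effect in the Dimensionator set, flagged for Niagara mesh particles.
# Two noise textures pan horizontally at different speeds, are summed, masked by a static distortion texture, then added to the base black hole texture.
# Particle color tints the result through an emissive adjustment function, while particle alpha scales it into a clamped, depth-faded opacity.
# The distortion texture only multiplies the noise; it does not offset UVs.
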Material /DealershipSharedLibrary/Assets/Sets/Dimensionator/Materials/M_Dimensionator_BlackHole
graph {
 properties {
  bAutomaticallySetUsageInEditor=false
  bUsedWithNiagaraMeshParticles=true
  blendMode=BLEND_Additive
  shadingModel=MSM_Unlit
  twoSided=true
  usedShadingModels=Unlit
 }
 nodes [
  n1 = TextureSample(UVs=Panner(Coordinate=TextureCoordinate() {vTiling=2}) {speedX=-0.1}) {texture=/DealershipSharedLibrary/Assets/Sets/Dimensionator/Textures/T_Dimensionator_CatFur_Noise}
  n2 = TextureSample(UVs=Panner(Coordinate=TextureCoordinate() {vTiling=2}) {speedX=-0.35}) {samplerType=SAMPLERTYPE_LinearColor, texture=/DealershipSharedLibrary/FX/Textures/Noise/T_VCVFX_CloudyNoise}
  n3 = TextureSample() {texture=/DealershipSharedLibrary/Assets/Sets/Dimensionator/Textures/T_Dimensionator_Blackhole_Distortion}
  n4 = Add(
   A=TextureSample() {texture=/DealershipSharedLibrary/Assets/Sets/Dimensionator/Textures/T_Dimensionator_BlackHole}.RGB
   B=Multiply(A=Add(A=@n1.RGB, B=@n2.RGB), B=Multiply(A=@n3.RGB))
  )
  n5 = ParticleColor()
  MaterialFunctionCall(Emissive=Multiply(A=@n4, B=@n5.RGB)) {materialFunction=&MF_FX_EmissiveAdjustments}
  MaterialFunctionCall(Opacity In=Clamp(None=Multiply(A=@n4, B=@n5.A)), DepthFade Distance=Constant() {r=1000}) {materialFunction=&MF_VCVFX_DepthFade}
 ]
}
