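# Additive unlit Niagara sprite material for the Dimensionator set's meteor smoke.
# A panned swirl flow texture distorts two scrolling noise samples, one also offset by camera direction; their heavily boosted product multiplies two panning energy wisp layers.
# A flame-cloud texture and particle alpha mask the result, while particle color feeds an emissive adjustment function.
# A dynamic parameter exposes noise tiling, speed, distortion and wisp tiling per particle.
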Material /DealershipSharedLibrary/Assets/Sets/Dimensionator/Materials/M_Dimensionator_MeteorSmoke
graph {
 properties {
  bAutomaticallySetUsageInEditor=false
  bUsedWithNiagaraSprites=true
  blendMode=BLEND_Additive
  shadingModel=MSM_Unlit
  usedShadingModels=Unlit
 }
 nodes [
  n1 = ParticleColor()
  MaterialFunctionCall(Emissive=@n1.RGB) {materialFunction=&MF_FX_EmissiveAdjustments}
  n2 = DynamicParameter() {paramNames [Noise Tile, Speed, Distort, Wisp TIle]}
  n3 = TextureCoordinate() {uTiling=0.75, vTiling=0.75}
  n4 = Multiply(A=@n2.Speed, B=Time())
  n5 = TextureSample(UVs=Panner(Coordinate=Multiply(A=@n3, B=@n2.Noise Tile), Time=@n4)) {texture=/Game/Global/Textures/Default/Flowmaps/T_SwirlFlow_03}
  n6 = Add(
   A=Multiply(A=TextureCoordinate() {uTiling=4, vTiling=4}, B=@n2.Noise Tile)
   B=Multiply(A=@n2.Distort, B=ComponentMask(None=@n5.RGB) {g=true, r=true})
  )
  n7 = ComponentMask(None=MaterialFunctionCall() {materialFunction=&CameraDirectionVector}.Result) {g=true, r=true}
  n8 = TextureSample(UVs=Panner(Coordinate=Add(A=@n6, B=@n7), Time=@n4) {speedX=-0.5, speedY=-0.13}) {texture=/DealershipSharedLibrary/FX/Textures/Noise/T_VCVFX_NoiseGeneric01Pack_RGB}
  n9 = TextureSample(UVs=Panner(Coordinate=@n6, Time=@n4) {speedX=-0.3, speedY=-0.1}) {texture=&T_VCVFX_NoiseGeneric01Pack_RGB}
  n10 = Multiply(A=Clamp(None=Multiply(A=@n8.B, B=@n9.B))) {constB=10000}
  n11 = Multiply(A=TextureCoordinate() {uTiling=2, vTiling=2}, B=@n2.Wisp TIle)
  n12 = Multiply(A=TextureCoordinate(), B=@n2.Wisp TIle)
  n13 = Multiply(
   A=TextureSample(UVs=Panner(Coordinate=@n11, Time=@n4) {speedX=0.2, speedY=-0.1}) {texture=/DealershipSharedLibrary/FX/Textures/Noise/T_VCVFX_Energy_Alpha}.RGB
   B=TextureSample(UVs=Panner(Coordinate=@n12, Time=@n4) {speedX=-0.05, speedY=-0.1}) {texture=&T_VCVFX_Energy_Alpha}.RGB
  )
  Multiply(
   A=@n1.A
   B=Multiply(A=TextureSample() {texture=/DealershipSharedLibrary/FX/Textures/Gradient_Masks/T_VCVFX_FlameCloud_02}.G, B=Multiply(A=@n10, B=@n13))
  )
 ]
}
