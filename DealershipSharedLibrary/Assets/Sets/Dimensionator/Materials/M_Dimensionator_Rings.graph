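# Translucent unlit material for Niagara mesh particles, evidently the ring effects of the Dimensionator set.
# Two stacks of summed, panning noise and streak textures, scrolling at different speeds and directions, blend through the Absorb dynamic parameter.
# The blend, tinted by particle color and scaled by the Bright parameter, drives emissive through an adjustment function.
# Opacity multiplies it by particle alpha and an asymmetric V-gradient edge mask, then applies depth fade.
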Material /DealershipSharedLibrary/Assets/Sets/Dimensionator/Materials/M_Dimensionator_Rings
graph {
 properties {
  bAutomaticallySetUsageInEditor=false
  bUsedWithNiagaraMeshParticles=true
  blendMode=BLEND_Translucent
  shadingModel=MSM_Unlit
  usedShadingModels=Unlit
 }
 nodes [
  n1 = DynamicParameter() {paramNames [Absorb,Bright,Param3,Param4]}
  n2 = ParticleColor()
  n3 = TextureCoordinate() {uTiling=2}
  n4 = Panner(Coordinate=TextureCoordinate()) {speedX=-1, speedY=0.5}
  n5 = Add(
   A=TextureSample(UVs=Panner(Coordinate=@n3) {speedX=-1, speedY=0.5}) {texture=/DealershipSharedLibrary/FX/Textures/Noise/T_VCVFX_CloudNoise}.RGB
   B=TextureSample(UVs=@n4) {texture=/DealershipSharedLibrary/FX/Textures/Noise/T_VCVFX_Water01_A}.RGB
  )
  n6 = TextureCoordinate() {vTiling=3}
  n7 = Multiply(A=TextureSample(UVs=Panner(Coordinate=@n6) {speedX=-1, speedY=1}) {texture=/DealershipSharedLibrary/Assets/Sets/Dimensionator/Textures/T_Dimensionator_NothrL_Prueba}.B) {constB=3.5}
  n8 = TextureSample(UVs=@n4) {texture=/DealershipSharedLibrary/Assets/Sets/Dimensionator/Textures/T_Dimensionator_NoiseMask}
  n9 = TextureCoordinate() {vTiling=2}
  n10 = TextureCoordinate()
  n11 = Add(
   A=TextureSample(UVs=Panner(Coordinate=@n9) {speedX=-0.1}) {texture=&T_VCVFX_CloudNoise}.RGB
   B=TextureSample(UVs=Panner(Coordinate=@n10) {speedX=-0.3}) {texture=&T_VCVFX_Water01_A}.RGB
  )
  n12 = Multiply(A=TextureSample(UVs=Panner(Coordinate=@n6) {speedX=-1}) {texture=/DealershipSharedLibrary/FX/Textures/TexturePackage/T_VCVFX_Boost_GreenAir_Trailcombo2_RGB}.B) {constB=6}
  n13 = TextureSample(UVs=Panner(Coordinate=@n10) {speedX=-0.2, speedY=0.3}) {texture=&T_Dimensionator_NoiseMask}
  n14 = LinearInterpolate(
   A=Add(A=Add(A=@n5, B=@n7), B=Multiply(A=@n8.RGB) {constB=3})
   B=Add(A=Add(A=@n11, B=@n12), B=Multiply(A=@n13.RGB) {constB=3})
   Alpha=@n1.Absorb
  )
  MaterialFunctionCall(Emissive=Multiply(A=@n1.Bright, B=Multiply(A=@n2.RGB, B=@n14))) {materialFunction=&MF_FX_EmissiveAdjustments}
  n15 = Multiply(A=@n2.A, B=@n14)
  n16 = ComponentMask(None=TextureCoordinate()) {g=true}
  n17 = OneMinus(None=@n16)
  n18 = Multiply(A=Multiply(A=Power(Base=@n16) {constExponent=6}, B=Power(Base=@n17) {constExponent=1.05})) {constB=3.5}
  MaterialFunctionCall(Opacity In=Clamp(None=Multiply(A=@n15, B=@n18)), DepthFade Distance=Constant() {r=1000}) {materialFunction=&MF_VCVFX_DepthFade}
 ]
}
